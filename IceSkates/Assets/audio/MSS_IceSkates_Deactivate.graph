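# One-shot mono MetaSound source for the ice skates deactivate sound, using footstep 3P attenuation and the weapon foley sound class.
# A random-sound subgraph plays one of three waves with slight pitch and volume variation, passing through finish and duration outputs.
# Its audio runs through a float multiply to the output.
# An affiliation-mix node reads character affiliation, but no connection from it to the multiply is shown.
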MetaSoundSource /IceSkates/Assets/audio/MSS_IceSkates_Deactivate
inputs [
 Character.Affiliation {type=Enum:FortAudioAffiliation}
 Character.Health {type=Float}
 Character.UniqueID {type=Int32}
 MaxPitch {type=Float}
 MaxVolume {type=Float}
 MinPitch {type=Float}
 MinVolume {type=Float}
 Pitch {type=Float}
 Seed {type=Int32}
 Shared State Behavior {type=Enum:SharedStateBehaviorType}
 Sounds {type=WaveAsset:Array}
 UE.Source.OnPlay {type=Trigger}
 Volume {type=Float}
]
interfaces [Fort.Character:1.0, UE.OutputFormat.Mono:1.0, UE.Source.OneShot:1.0, UE.Source:1.1]
outputs [
 "UE.OutputFormat.Mono.Audio:0" {type=Audio}
 UE.Source.OneShot.OnFinished {type=Trigger}
 WaveDuration {type=Time}
]
pages [Default {
  input_defaults {
   Character.Affiliation=0
   Character.Health=0
   Character.UniqueID=0
   MaxPitch=1.05
   MaxVolume=1.05
   MinPitch=0.95
   MinVolume=0.95
   Pitch=1
   Seed=-1
   Shared State Behavior=1
   UE.Source.OnPlay=null
   Volume=0.75
   Sounds [
    &/IceSkates_Deactivate_01
    &/IceSkates_Deactivate_02
    &/IceSkates_Deactivate_03
   ]
  }
  nodes [
   &MSP_AffiliationMixFloat(Affiliation=^Character.Affiliation) {also {defaults {Local=0.65, Non-Enemy=0.5}}}
   MSS_PlayRandomSound_Mono_0 = &MSS_PlayRandomSound_Mono(
    MaxPitch=^MaxPitch
    MaxVolume=^MaxVolume
    MinPitch=^MinPitch
    MinVolume=^MinVolume
    Pitch=^Pitch
    Seed=^Seed
    Shared State Behavior=^Shared State Behavior
    Sounds=^Sounds
    UE.Source.OnPlay=^UE.Source.OnPlay
    Volume=^Volume
   )
   Multiply_0 = UE.Multiply:Audio by Float(PrimaryOperand=@MSS_PlayRandomSound_Mono_0."UE.OutputFormat.Mono.Audio:0")
  ]
  links [
   {dst="@UE.Source.OneShot.OnFinished", src=MSS_PlayRandomSound_Mono_0.UE.Source.OneShot.OnFinished}
   {dst="@WaveDuration", src=MSS_PlayRandomSound_Mono_0.WaveDuration}
   {dst="@UE.OutputFormat.Mono.Audio:0", src=Multiply_0.Out}
  ]
 }]
properties {
 assetImportData=&self:AssetImportData
 attenuationSettings=/Game/Sounds/Fort_Player/Footsteps/Settings/Fort_Player_FS_3P_Att
 maxDistance=3584
 sampleRate=48000
 soundClassObject=&Fort_WeaponFoley
}
