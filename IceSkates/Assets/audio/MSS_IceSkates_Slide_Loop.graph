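# Mono MetaSound source for the ice skate slide loop, a player foley sound that follows movement speed.
# Movement.Speed is range-mapped and smoothed with InterpTo, then drives both the wave player's pitch shift and its gain.
# An ADSR envelope shapes the output, attacking on play or Activate, releasing on Deactivate and firing OnFinished when done.
# The wave player starts only on play, and the separate Speed input is unused.
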MetaSoundSource /IceSkates/Assets/audio/MSS_IceSkates_Slide_Loop
inputs [
 Activate {type=Trigger}
 Deactivate {type=Trigger}
 Movement.Speed {type=Float}
 Movement.SpeedAzimuth {type=Float}
 Movement.SpeedElevation {type=Float}
 Speed {type=Float}
 UE.Source.OnPlay {type=Trigger}
 Wave Asset {type=WaveAsset}
]
interfaces [Fort.Movement:1.0, UE.OutputFormat.Mono:1.0, UE.Source.OneShot:1.0, UE.Source:1.1]
outputs [
 "UE.OutputFormat.Mono.Audio:0" {type=Audio}
 UE.Source.OneShot.OnFinished {type=Trigger}
]
pages [Default {
  input_defaults {
   Activate=null
   Deactivate=null
   Movement.Speed=0
   Movement.SpeedAzimuth=0
   Movement.SpeedElevation=0
   Speed=0
   UE.Source.OnPlay=null
   Wave Asset=&/IceSkates_SlideLoop_Player_01
  }
  nodes [
   TriggerAnyTriggerAny2_0 = "TriggerAny.Trigger Any (2)"(In 0=^Activate, In 1=^UE.Source.OnPlay)
   ADSREnvelopeADSREnvelope_0 = ADSR Envelope.ADSR Envelope:Audio(Trigger Release=^Deactivate, Trigger Attack=@TriggerAnyTriggerAny2_0.Out) {also {defaults {Attack Time=1, Decay Time=3, Release Time=0.2}}}
   n1 = UE.Reroute(Value=UE.InterpTo:Audio(Target=MapRange.MapRange:Float(In=^Movement.Speed) {also {defaults {In Range B=1200, Out Range B=0.5}}}.Out Value) {also {defaults {Interp Time=0.2}}}.Value)
   n2 = UE.Reroute(Value=@n1.Value)
   WavePlayer_0 = UE.Wave Player:Mono(
    Play=^UE.Source.OnPlay
    Wave Asset=^Wave Asset
    Pitch Shift=MapRange.MapRange:Float(In=@n1.Value) {also {defaults {Out Range A=0.9, Out Range B=1.1}}}.Out Value
   ) {also {defaults {Loop="<MetaSoundBuilderResult.SUCCEEDED: 0>"}, version [1,1]}}
   Multiply_0 = UE.Multiply:Audio(
    PrimaryOperand=@ADSREnvelopeADSREnvelope_0.Out Envelope
    AdditionalOperands=UE.Multiply:Audio by Float(AdditionalOperands=MapRange.MapRange:Float(In=@n2.Value).Out Value, PrimaryOperand=@WavePlayer_0.Out Mono).Out
   )
  ]
  links [{dst="@UE.Source.OneShot.OnFinished", src=ADSREnvelopeADSREnvelope_0.On Done}, {dst="@UE.OutputFormat.Mono.Audio:0", src=Multiply_0.Out}]
 }]
properties {
 assetImportData=&self:AssetImportData
 attenuationSettings=/Game/Sounds/Fort_Player/Footsteps/Settings/Fort_Player_FS_3P_Att
 maxDistance=3584
 pitch=2
 sampleRate=48000
 soundClassObject=&Fort_Player_Foley
 modulationSettings {highpassRouting=Union, highpassModulationDestination {value=300}}
}
