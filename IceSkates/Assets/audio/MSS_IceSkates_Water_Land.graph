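# MetaSound source for the ice-skate landing-on-water footstep sound in the IceSkates content.
# It is a preset of MSS_PlayRandomSound_Mono, so the parent supplies the graph and this asset only sets inputs.
# It feeds four water-land wave variations with slight random pitch and volume ranges, as a mono one-shot.
# It uses player footstep third-person attenuation and the player foley sound class.
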MetaSoundSource /IceSkates/Assets/audio/MSS_IceSkates_Water_Land
preset_parent=&MSS_PlayRandomSound_Mono
inputs [
 MaxPitch {type=Float}
 MaxVolume {type=Float}
 MinPitch {type=Float}
 MinVolume {type=Float}
 Pitch {type=Float}
 Seed {type=Int32}
 Shared State Behavior {type=Enum:SharedStateBehaviorType}
 Sounds {type=WaveAsset:Array}
 UE.Source.OnPlay {type=Trigger}
 Volume {type=Float}
]
interfaces [UE.OutputFormat.Mono:1.0, UE.Source.OneShot:1.0, UE.Source:1.1]
outputs [
 "UE.OutputFormat.Mono.Audio:0" {type=Audio}
 UE.Source.OneShot.OnFinished {type=Trigger}
 WaveDuration {type=Time}
]
pages [Default {input_defaults {
   MaxPitch=1.05
   MaxVolume=1.05
   MinPitch=0.95
   MinVolume=0.95
   Pitch=1
   Seed=-1
   Shared State Behavior=1
   UE.Source.OnPlay=null
   Volume=0.75
   Sounds [
    &/IceSkates_Footsteps_Water_Land_01
    &/IceSkates_Footsteps_Water_Land_02
    &/IceSkates_Footsteps_Water_Land_03
    &/IceSkates_Footsteps_Water_Land_04
   ]
  }}]
properties {
 assetImportData=&self:AssetImportData
 attenuationSettings=/Game/Sounds/Fort_Player/Footsteps/Settings/Fort_Player_FS_3P_Att
 maxDistance=3584
 sampleRate=48000
 soundClassObject=&Fort_Player_Foley
}
