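# Sound cue for equipping ice skates, playing one of four equip foley waves.
# An equal-weight random node picks the wave and feeds a mixer with only that input connected, then a root modulator slightly varies pitch and volume.
# It plays at reduced volume in the weapon foley sound class, using weapon drop attenuation.
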SoundCue /IceSkates/Assets/audio/SC_IceSkates_Equip
root=Modulator_0
nodes [
 {id=Mixer_0, children [Random_1,null,null], properties {inputVolume [1,1,1]}}
 {id=Modulator_0, children [Mixer_0], properties {pitchMax=1.1, pitchMin=0.9, volumeMax=1.1, volumeMin=0.9}}
 {id=Random_1, children [WavePlayer_0,WavePlayer_1,WavePlayer_2,WavePlayer_3], properties {weights [1,1,1,1]}}
 {id=WavePlayer_0, properties {soundWaveAssetPtr=&/IceSkates_Equip_01}}
 {id=WavePlayer_1, properties {soundWaveAssetPtr=&/IceSkates_Equip_02}}
 {id=WavePlayer_2, properties {soundWaveAssetPtr=&/IceSkates_Equip_03}}
 {id=WavePlayer_3, properties {soundWaveAssetPtr=&/IceSkates_Equip_04}}
]
properties {
 attenuationSettings=&Fort_Weapon_Drop_Att
 maxDistance=3000
 soundClassObject=&Fort_WeaponFoley
 volumeMultiplier=0.5
}
