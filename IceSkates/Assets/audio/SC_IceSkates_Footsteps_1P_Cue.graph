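# Sound cue for first-person ice skate footsteps, played at reduced volume with a first-person walk footstep attenuation.
# A quality-level switch picks a random pool of seven, four or two waves, with the last two levels sharing the two-wave pool.
# A modulator at the root slightly varies pitch and volume on each play.
# Despite the 1P name, it is assigned to the 3P enemy player foley sound class.
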SoundCue /IceSkates/Assets/audio/SC_IceSkates_Footsteps_1P_Cue
root=Modulator_0
nodes [
 {id=Modulator_0, children [QualityLevel_1], properties {pitchMax=1.1, pitchMin=0.9, volumeMax=1.1, volumeMin=0.9}}
 {id=QualityLevel_1, children [Random_4,Random_5,Random_3,Random_3]}
 {id=Random_3, children [WavePlayer_24, WavePlayer_25], properties {bShouldExcludeFromBranchCulling=true, weights [1,1]}}
 {
  id=Random_4
  children [WavePlayer_13, WavePlayer_14, WavePlayer_15, WavePlayer_16, WavePlayer_17, WavePlayer_18, WavePlayer_19]
  properties {weights [1,1,1,1,1,1,1]}
 }
 {id=Random_5, children [WavePlayer_20, WavePlayer_21, WavePlayer_22, WavePlayer_23], properties {weights [1,1,1,1]}}
 {id=WavePlayer_13, properties {soundWaveAssetPtr=&/IceSkates_Footsteps_V1_07}}
 {id=WavePlayer_14, properties {soundWaveAssetPtr=&/IceSkates_Footsteps_V1_06}}
 {id=WavePlayer_15, properties {soundWaveAssetPtr=&/IceSkates_Footsteps_V1_05}}
 {id=WavePlayer_16, properties {soundWaveAssetPtr=&/IceSkates_Footsteps_V1_04}}
 {id=WavePlayer_17, properties {soundWaveAssetPtr=&/IceSkates_Footsteps_V1_03}}
 {id=WavePlayer_18, properties {soundWaveAssetPtr=&/IceSkates_Footsteps_V1_02}}
 {id=WavePlayer_19, properties {soundWaveAssetPtr=&/IceSkates_Footsteps_V1_01}}
 {id=WavePlayer_20, properties {soundWaveAssetPtr=&IceSkates_Footsteps_V1_07}}
 {id=WavePlayer_21, properties {soundWaveAssetPtr=&IceSkates_Footsteps_V1_06}}
 {id=WavePlayer_22, properties {soundWaveAssetPtr=&IceSkates_Footsteps_V1_05}}
 {id=WavePlayer_23, properties {soundWaveAssetPtr=&IceSkates_Footsteps_V1_04}}
 {id=WavePlayer_24, properties {soundWaveAssetPtr=&IceSkates_Footsteps_V1_07}}
 {id=WavePlayer_25, properties {soundWaveAssetPtr=&IceSkates_Footsteps_V1_06}}
]
properties {
 attenuationSettings=/Game/Sounds/Attenuation/Fort_Player_Footsteps_1P_Walk_Athena_Att
 maxDistance=2097152
 soundClassObject=&Fort_Player_Foley_3P_Enemy_Athena
 volumeMultiplier=0.2
}
