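# Sound cue for the IceSkates jump or begin-fall foley, judging by its name, routed to the player foley sound class.
# A random node picks evenly among four skate pickup waves, then a modulator varies pitch and volume slightly.
# The cue plays quieter and higher-pitched, using a third-person player jump attenuation asset.
# A concurrency override caps simultaneous instances and sets a short retrigger time.
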SoundCue /IceSkates/Assets/audio/SC_IceSkates_JumpOrBeginFall_Cue
root=Modulator_0
nodes [
 {id=Modulator_0, children [Random_1], properties {pitchMax=1.1, pitchMin=0.9, volumeMax=1.1, volumeMin=0.9}}
 {id=Random_1, children [WavePlayer_2,WavePlayer_3,WavePlayer_4,WavePlayer_5], properties {weights [1,1,1,1]}}
 {id=WavePlayer_2, properties {soundWaveAssetPtr=&/IceSkates_Pickup_04}}
 {id=WavePlayer_3, properties {soundWaveAssetPtr=&/IceSkates_Pickup_03}}
 {id=WavePlayer_4, properties {soundWaveAssetPtr=&/IceSkates_Pickup_02}}
 {id=WavePlayer_5, properties {soundWaveAssetPtr=&/IceSkates_Pickup_01}}
]
properties {
 attenuationSettings=/Game/Sounds/Attenuation/Fort_Player_Jump_3P_Athena_Att
 bOverrideConcurrency=true
 maxDistance=3584
 pitchMultiplier=1.3
 soundClassObject=&Fort_Player_Foley
 volumeMultiplier=0.35
 concurrencyOverrides {maxCount=3, retriggerTime=0.1}
}
