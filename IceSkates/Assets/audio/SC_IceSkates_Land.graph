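# SC_IceSkates_Land is a Sound Cue that plays an ice skate slide impact, named for landing.
# A Random node picks among four slide impact waves with equal weight, and a Modulator root slightly varies pitch and volume on each play.
# The cue plays at reduced volume, borrowing a weapon-drop attenuation asset and the loot collect sound class.
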SoundCue /IceSkates/Assets/audio/SC_IceSkates_Land
root=Modulator_0
nodes [
 {id=Modulator_0, children [Random_0], properties {pitchMax=1.1, pitchMin=0.9, volumeMax=1.1, volumeMin=0.9}}
 {id=Random_0, children [WavePlayer_0,WavePlayer_1,WavePlayer_2,WavePlayer_3], properties {weights [1,1,1,1]}}
 {id=WavePlayer_0, properties {soundWaveAssetPtr=&/IceSkates_SlideImpact_02}}
 {id=WavePlayer_1, properties {soundWaveAssetPtr=&/IceSkates_SlideImpact_03}}
 {id=WavePlayer_2, properties {soundWaveAssetPtr=&/IceSkates_SlideImpact_04}}
 {id=WavePlayer_3, properties {soundWaveAssetPtr=&/IceSkates_SlideImpact_05}}
]
properties {
 attenuationSettings=&Fort_Weapon_Drop_Att
 maxDistance=3000
 soundClassObject=&Fort_Loot_Collect
 volumeMultiplier=0.35
}
