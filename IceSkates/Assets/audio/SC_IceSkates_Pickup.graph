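# Sound cue for picking up the Ice Skates item, using the loot-collect sound class and weapon pickup attenuation.
# A mixer layers one of four random pickup waves, boosted in volume, with one of three random ice warning-crack waves at reduced volume.
# A root modulator applies slight random pitch and volume variation to the combined result on each play.
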SoundCue /IceSkates/Assets/audio/SC_IceSkates_Pickup
root=Modulator_0
nodes [
 {id=Mixer_0, children [Random_1,Random_4], properties {inputVolume [1.25,0.75]}}
 {id=Modulator_0, children [Mixer_0], properties {pitchMax=1.1, pitchMin=0.9, volumeMax=1.1, volumeMin=0.9}}
 {id=Random_1, children [WavePlayer_4,WavePlayer_5,WavePlayer_6,WavePlayer_7], properties {weights [1,1,1,1]}}
 {id=Random_4, children [WavePlayer_13, WavePlayer_14, WavePlayer_15], properties {weights [1,1,1]}}
 {id=WavePlayer_13, properties {soundWaveAssetPtr=&/IceSkates_WarningCrack_01}}
 {id=WavePlayer_14, properties {soundWaveAssetPtr=&/IceSkates_WarningCrack_02}}
 {id=WavePlayer_15, properties {soundWaveAssetPtr=&/IceSkates_WarningCrack_03}}
 {id=WavePlayer_4, properties {soundWaveAssetPtr=&/IceSkates_Pickup_04}}
 {id=WavePlayer_5, properties {soundWaveAssetPtr=&/IceSkates_Pickup_03}}
 {id=WavePlayer_6, properties {soundWaveAssetPtr=&/IceSkates_Pickup_02}}
 {id=WavePlayer_7, properties {soundWaveAssetPtr=&/IceSkates_Pickup_01}}
]
properties {
 attenuationSettings=&Fort_Weapon_Pickup_Att
 maxDistance=3500
 soundClassObject=&Fort_Loot_Collect
}
