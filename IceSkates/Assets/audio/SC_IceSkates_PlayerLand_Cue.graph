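# Sound cue for a player landing on ice skates, built from slide impact samples.
# A Random node picks one of five equally weighted impact waves and feeds a Modulator at its default settings for variation.
# It borrows small melee weapon attenuation and the melee weapon sound class, at reduced volume.
# A concurrency override blocks new instances once the limit is reached instead of stopping playing ones.
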SoundCue /IceSkates/Assets/audio/SC_IceSkates_PlayerLand_Cue
root=Modulator_2
nodes [
 {id=Modulator_2, children [Random_1]}
 {id=Random_1, children [WavePlayer_4,WavePlayer_5,WavePlayer_6,WavePlayer_7,WavePlayer_8], properties {weights [1,1,1,1,1]}}
 {id=WavePlayer_4, properties {soundWaveAssetPtr=&/IceSkates_SlideImpact_01}}
 {id=WavePlayer_5, properties {soundWaveAssetPtr=&/IceSkates_SlideImpact_02}}
 {id=WavePlayer_6, properties {soundWaveAssetPtr=&/IceSkates_SlideImpact_03}}
 {id=WavePlayer_7, properties {soundWaveAssetPtr=&/IceSkates_SlideImpact_04}}
 {id=WavePlayer_8, properties {soundWaveAssetPtr=&/IceSkates_SlideImpact_05}}
]
properties {
 attenuationSettings=&Fort_Weapon_Melee_Sml_Att
 bOverrideConcurrency=true
 maxDistance=2800
 soundClassObject=&Fort_WeaponMelee_PSM
 volumeMultiplier=0.6
 attenuationOverrides {falloffDistance=2600}
 concurrencyOverrides {resolutionRule=PreventNew}
}
