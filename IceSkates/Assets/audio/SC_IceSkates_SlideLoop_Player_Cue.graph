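# Sound cue for the looping ice skate slide sound on a player, using third-person sprint footstep attenuation and the enemy foley sound class.
# A looping wave player feeds a continuous modulator, where a Speed parameter drives volume and shifts pitch within a narrow range.
# Volume modulation defaults to zero when Speed is not supplied.
# Concurrency is limited per owner, with overall volume scaled well down and pitch scaled up.
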SoundCue /IceSkates/Assets/audio/SC_IceSkates_SlideLoop_Player_Cue
root=ModulatorContinuous_0
nodes [
 {
  id=ModulatorContinuous_0
  children [WavePlayer_0]
  properties {pitchModulationParams {maxOutput=1.1, minOutput=0.9, parameterName=Speed}, volumeModulationParams {default=0, parameterName=Speed}}
 }
 {id=WavePlayer_0, properties {bLooping=true, soundWaveAssetPtr=&/IceSkates_SlideLoop_Player_01}}
]
properties {
 attenuationSettings=/Game/Sounds/Attenuation/Fort_Player_Footsteps_3P_Sprint_Athena_Att
 maxDistance=3584
 pitchMultiplier=2
 soundClassObject=&Fort_Player_Foley_3P_Enemy_Athena
 volumeMultiplier=0.1
 concurrencyOverrides {bLimitToOwner=true}
}
